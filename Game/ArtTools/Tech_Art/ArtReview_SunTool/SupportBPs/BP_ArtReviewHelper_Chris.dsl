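; Art review helper actor that swaps the season 14 TimeOfDayManager sun for a false directional light and previews material channels.
; TODM_Off hides the real sun and copies its intensity, color and transform to the false rig.
; Custom events set the hour on the manager, hide or show it, and rotate the false sun by heading or altitude.
; Visualizer functions toggle per-channel post-process components, while Emissive disables diffuse and specular show flags.
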(event AnyKey
  (params (Key :type Key%20Structure)))

(event EventBeginPlay
  (bind _returnvalue (Game|GetPlayerController 0))
  (Utilities|IsValid _returnvalue
    (:Is%20Valid
      (Input|MouseInput|BindEventtoOnClicked _returnvalue (delegate Custom|CustomEvent_0)))))

(event Custom|CustomEvent_0
  (params (TouchedActor :type Actor%20Object%20Reference) (ButtonPressed :type Key%20Structure))
  (Development|PrintString "Binding Worked"))

(event Custom|TODM_On
  ;[1 1 1] Toggle Real/False Sun Lighting
  (Development|PrintString "TODM On")
  (bind _returnvalue (Actor|GetActorOfClass "/Game/TimeOfDay/TODM/BR/s11/TODM_BR_s14.TODM_BR_s14_C"))
  (Development|PrintString (Utilities|GetDisplayName _returnvalue) false)
  (Rendering|SetVisibility (Variables|/Script/FortniteGame.FortTimeOfDayManager|GetDirectionalLightComponent (stale Custom|TODM_Off _returnvalue)) true false)
  (Rendering|SetVisibility (Variables|Default|GetScene4) false true))

(event Collision|EventActorBeginOverlap
  (params (OtherActor :type Actor%20Object%20Reference)))

(event Custom|TODM_Off
  ;[1 1 1] Toggle Real/False Sun Lighting
  (Development|PrintString "TODM Off")
  (bind _returnvalue (Actor|GetActorOfClass "/Game/TimeOfDay/TODM/BR/s11/TODM_BR_s14.TODM_BR_s14_C"))
  (Development|PrintString (Utilities|GetDisplayName _returnvalue) false)
  (Rendering|SetVisibility (Variables|/Script/FortniteGame.FortTimeOfDayManager|GetDirectionalLightComponent _returnvalue) false false)
  (Rendering|SetVisibility (Variables|Default|GetScene4) true true)
  (CallFunction|CopyDirectional (Variables|/Script/FortniteGame.FortTimeOfDayManager|GetDirectionalLightComponent _returnvalue) (Variables|Default|GetDirectionalLight)))

(event Custom|Hide%20Sun%20Systems
  ;[1 1 1] Toggle Real/False Sun Lighting
  (Development|PrintString "HIDE TODM")
  (bind _returnvalue (Actor|GetActorOfClass "/Game/TimeOfDay/TODM/BR/s11/TODM_BR_s14.TODM_BR_s14_C"))
  (Rendering|SetActorHiddenInGame (stale Custom|TODM_Off _returnvalue) true))

(event Custom|Show%20Sun%20Systems
  ;[1 1 1] Toggle Real/False Sun Lighting
  (Development|PrintString "Show TODM")
  (bind _returnvalue (Actor|GetActorOfClass "/Game/TimeOfDay/TODM/BR/s11/TODM_BR_s14.TODM_BR_s14_C"))
  (Rendering|SetActorHiddenInGame _returnvalue false))

(event Custom|UpdateHour
  (params (hour :type Float%20%28double-precision%29))
  ;[1 1 1] Pass along time updates
  (Development|PrintString (Utilities|String|Append "In Helper BP: " (Utilities|String|ToString(Float) hour)) false)
  (bind _output_get (Variables|Sun|SetHour hour))
  (bind _returnvalue (Actor|GetActorOfClass "/Game/TimeOfDay/TODM/BR/s11/TODM_BR_s14.TODM_BR_s14_C"))
  (TimeOfDay|SetTimeOfDay _returnvalue _output_get :!class /Script/FortniteGame.FortTimeOfDayManagerInterface)
  (TimeOfDay|SetSpecificTimeOfDayinEditor _returnvalue _output_get))

(event Custom|UpdateFalseRotation
  (params (Forward%20Vector :type Vector) (SliderValue :type Float%20%28double-precision%29))
  ;[1 1 1] ROTATION
  (Transformation|SetWorldRotation (Variables|Default|GetRotation%20Center) (Math|Rotator|MakeRotator (.roll (Transformation|GetWorldRotation (Variables|Default|GetRotation%20Center))) (.pitch (Transformation|GetWorldRotation (Variables|Default|GetRotation%20Center))) (.yaw (Math|Rotator|FindLookatRotation (Transformation|GetForwardVector (Variables|Default|GetRotation%20Center)) (+ (Math|Vector|Normalize (Math|Vector|MakeVector (.x (Transformation|GetWorldLocation (Variables|Default|GetRotation%20Center))) (.y (Transformation|GetWorldLocation (Variables|Default|GetRotation%20Center))))) (* (Math|Vector|Normalize (Math|Vector|MakeVector (.x Forward%20Vector) (.y Forward%20Vector))) -5e+01))))) false false) ; Rotate around the Z
)

(event Custom|UpdateAltitudeRotation
  (params (Rotation :type Float%20%28double-precision%29))
  ;[1 1 1] Update the false sun inclination
  (Transformation|SetWorldRotation (Variables|Default|GetRotation%20Center) (Math|Rotator|MakeRotator (.roll (Transformation|GetWorldRotation (Variables|Default|GetRotation%20Center))) (* Rotation -1.0) (.yaw (Transformation|GetWorldRotation (Variables|Default|GetRotation%20Center)))) false false))

(event Custom|ForceSetRotation
  (params (Rotation :type Vector))
  ;[1 1 1] Update the false sun inclination
  (Transformation|SetWorldRotation (Variables|Default|GetRotation%20Center) (Math|Rotator|MakeRotator (.x Rotation) (.y Rotation) (.z Rotation)) false false))

(fn BaseColor
  (RenderPassVisualizers|Reset)
  (Variables|/Script/Engine.PostProcessComponent|SetbEnabled true (Variables|Default|GetPostProcess_baseColor)))

(fn copyDirectional
  (params (source :type Light%20Component%20Base%20Object%20Reference) (target :type Directional%20Light%20Component%20Object%20Reference))
  (bind _returnvalue (Rendering|Components|Light|GetLightColor source))
  (bind _intensity (Variables|/Script/Engine.LightComponentBase|GetIntensity source))
  (Rendering|Components|Light|SetIntensity target _intensity)
  (Development|PrintString (Utilities|String|Append "Intensity: " (Utilities|String|ToString(Float) _intensity)))
  (Rendering|Components|Light|SetLightColor target _returnvalue true)
  (Development|PrintString (Utilities|String|Append "Color (SRGB):" (Utilities|String|ToString(LinearColor) _returnvalue)))
  (Transformation|SetWorldTransform (Variables|Default|GetRotation%20Center) (Math|Transform|MakeTransform (.location (Transformation|GetWorldTransform source)) (.rotation (Transformation|GetWorldTransform source)) (.scale (Transformation|GetWorldTransform source))) false true))

(fn Emissive
  (RenderPassVisualizers|Reset)
  (Development|ExecuteConsoleCommand "ShowFlag.Diffuse 0")
  (Development|ExecuteConsoleCommand "ShowFlag.Specular 0"))

(fn Metallic
  (RenderPassVisualizers|Reset)
  (Variables|/Script/Engine.PostProcessComponent|SetbEnabled true (Variables|Default|GetPostProcess_metallic)))

(fn Reset
  (Variables|/Script/Engine.PostProcessComponent|SetbEnabled false (Variables|Default|GetPostProcess_baseColor))
  (Variables|/Script/Engine.PostProcessComponent|SetbEnabled false (Variables|Default|GetPostProcess_specular))
  (Variables|/Script/Engine.PostProcessComponent|SetbEnabled false (Variables|Default|GetPostProcess_roughness))
  (Variables|/Script/Engine.PostProcessComponent|SetbEnabled false (Variables|Default|GetPostProcess_metallic))
  (Development|ExecuteConsoleCommand "ShowFlag.Diffuse 2")
  (Development|ExecuteConsoleCommand "ShowFlag.Specular 2"))

(fn Roughness
  (RenderPassVisualizers|Reset)
  (Variables|/Script/Engine.PostProcessComponent|SetbEnabled true (Variables|Default|GetPostProcess_roughness)))

(fn Specular
  (RenderPassVisualizers|Reset)
  (Variables|/Script/Engine.PostProcessComponent|SetbEnabled true (Variables|Default|GetPostProcess_specular)))

(fn ConstructionScript)
