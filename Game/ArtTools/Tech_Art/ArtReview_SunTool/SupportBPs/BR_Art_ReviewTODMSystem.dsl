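; Time-of-day manager Blueprint for art review that can manually override the sun direction used by anime shading.
; BeginPlay creates a virtual 3D HUD widget, showing it and sending default spacers only if the review helper actor exists.
; Custom events store override vectors and the toggle, forwarding changes to the HUD through interface messages.
; When enabled, the directional light update hook writes the stored vector to SunlightDirectionalVector in the global material parameter collection after calling the parent.
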(event EventBeginPlay
  (|CallParentFunction)
  (bind _returnvalue (UserInterface|CreateWidget "/Game/ArtTools/Tech_Art/ArtReview_SunTool/SupportWidgets/ArtReview_Virtual3DHudeManager.ArtReview_Virtual3DHudeManager_C"))
  (bind _output_get (Variables|Default|SetHudRef _returnvalue))
  (bind _returnvalue_2 (Actor|GetActorOfClass "/Game/ArtTools/Tech_Art/ArtReview_SunTool/SupportBPs/BP_ArtReviewHelper_Chris.BP_ArtReviewHelper_Chris_C"))
  (Utilities|IsValid _returnvalue_2
    (:Is%20Valid
      (UserInterface|Viewport|AddtoViewport _output_get 0)
      (CallFunction|UpdateSpacers 0.0 1e+01 3.103 0.448))))

(event EventTick
  (params (DeltaSeconds :type Float%20%28single-precision%29))
  (|CallParentFunction DeltaSeconds) ; This node is disabled and will not be called.\nDrag off pins to build functionality.
)

(event EventOnUpdateDirectionalLightForTimeOfDay
  (params (AboutToApplyDirectionalLightColor :type Linear%20Color%20Structure%20%28by%20ref%29))
  ;[1 1 1] Anime Sun Lighting Review Overrides
  (|CallParentFunction AboutToApplyDirectionalLightColor)
  (if (Variables|Default|GetSwitchManualAnimeShadeSunDirection)
    (Rendering|Material|SetVectorParameterValue "/Game/Packages/Fortress_SharedMaterials/GlobalMaterialParameters/FortniteMaterialParameters.FortniteMaterialParameters" "SunlightDirectionalVector" (Variables|Default|GetOverwrittenSunDirectionalVector) :!class /Script/Engine.KismetMaterialLibrary)))

(event Custom|UpdateSunOverrideValues
  (params (OverwritenSunDirectionalVector :type Linear%20Color%20Structure) (OverwrittenSunYDirection :type Linear%20Color%20Structure) (OverwrittenSunZDirection :type Linear%20Color%20Structure))
  ;[1 1 1] Anime Sun Lighting Review Overrides
  (Variables|Default|SetOverwrittenSunDirectionalVector OverwritenSunDirectionalVector)
  (Variables|Default|SetOverwrittenSunYDirection OverwrittenSunYDirection)
  (Variables|Default|SetOverwrittenSunZDirection OverwrittenSunZDirection))

(event Custom|SetManualAnimeSunDirection
  (params (SwitchManualAnimeShadeSunDirection :type Boolean))
  ;[1 1 1] Anime Sun Lighting Review Overrides
  (Variables|Default|SetSwitchManualAnimeShadeSunDirection SwitchManualAnimeShadeSunDirection)
  (Class|InterfaceVirtual3DDrawspace|UpdateOverWritten(Message) (Variables|Default|GetHudRef) (Variables|Default|GetSwitchManualAnimeShadeSunDirection) (Math|Conversions|ToVector(LinearColor) (Variables|Default|GetOverwrittenSunYDirection)) (Math|Conversions|ToVector(LinearColor) (Variables|Default|GetOverwrittenSunZDirection))))

(event Custom|UpdateSpacers
  (params (LeftSpacer :type Float%20%28double-precision%29) (RightSpacer :type Float%20%28double-precision%29) (TopSpacer :type Float%20%28double-precision%29) (BottomSpacer :type Float%20%28double-precision%29))
  ;[1 1 1] Passes along Spacer Updates for HUD
  (Class|InterfaceVirtual3DDrawspace|UpdateSpacers(Message) (Variables|Default|GetHudRef) LeftSpacer RightSpacer TopSpacer BottomSpacer)
  (Development|PrintString "Spacers Updated" false))

(fn ConstructionScript
  (|CallParentFunction))
